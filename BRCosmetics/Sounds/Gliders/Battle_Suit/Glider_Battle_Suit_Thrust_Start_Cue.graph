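# Sound cue for the Battle Suit glider's thrust-start sound.
# A random node picks one of three equally weighted thrust waves, and a modulator above it slightly varies the volume.
# Playback uses the glider thrust attenuation and concurrency settings and is assigned to the Fort_Gliders_Land sound class.
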SoundCue /BRCosmetics/Sounds/Gliders/Battle_Suit/Glider_Battle_Suit_Thrust_Start_Cue
root=Modulator_0
nodes [
 {id=Modulator_0, children [Random_0], properties {volumeMax=1, volumeMin=0.85}}
 {id=Random_0, children [WavePlayer_0,WavePlayer_2,WavePlayer_5], properties {weights [1,1,1]}}
 {id=WavePlayer_0, properties {soundWaveAssetPtr=&/Glider_Battle_Suit_Thrust_01}}
 {id=WavePlayer_2, properties {soundWaveAssetPtr=&/Glider_Battle_Suit_Thrust_02}}
 {id=WavePlayer_5, properties {soundWaveAssetPtr=&/Glider_Battle_Suit_Thrust_05}}
]
properties {
 attenuationSettings=&Fort_Player_Glider_Thrust_Athena_Att
 maxDistance=12500
 soundClassObject=&Fort_Gliders_Land
 volumeMultiplier=0.65
 concurrencySet [&Athena_Glider_Thrust_Conc]
}
