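# Sound cue for the looping thrust audio of the Battle Suit glider cosmetic.
# A single looping wave player feeds the Battle Suit loop wave straight to the output, with no mixing or modulation nodes.
# The cue lowers volume and uses player glider thrust attenuation, a cosmetic glider loop audio properties sheet, and a glider sound class.
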SoundCue /BRCosmetics/Sounds/Gliders/Battle_Suit/Glider_Battle_Suit_Thrust_loop_Cue
root=WavePlayer_0
nodes [{id=WavePlayer_0, properties {bLooping=true, soundWaveAssetPtr=&/Glider_Battle_Suit_Loop}}]
properties {
 attenuationSettings=&Fort_Player_Glider_Thrust_Athena_Att
 audioPropertiesSheet=&APS_Cosmetic_Glider_Loop
 maxDistance=12500
 soundClassObject=&Fort_Gliders_Land
 volumeMultiplier=0.3
}
